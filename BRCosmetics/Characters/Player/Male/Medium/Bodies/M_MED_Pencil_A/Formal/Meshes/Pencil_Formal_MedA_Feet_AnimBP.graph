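# Post-process style AnimBP for the Pencil formal medium male feet mesh that adds secondary motion to shoelace bows and strands.
# It copies the parent mesh pose and curves, then chains sprung, angle-limited Anim Dynamics nodes relative to each foot.
# Right foot binds r_bow_01 and l_strand_01 twice, so r_bow_02 and l_strand_02 have no node.
# The event graph has no logic.
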AnimBlueprint /BRCosmetics/Characters/Player/Male/Medium/Bodies/M_MED_Pencil_A/Formal/Meshes/Pencil_Formal_MedA_Feet_AnimBP
target_skeleton=&/Pencil_Formal_MedA_Feet_Skeleton
graphs [
 AnimGraph {nodes [
   n1 = Anim Dynamics - Bone: foot_l_shoelace_r_bow_01_dyn(
    ComponentPose=Local To Component(LocalPose=Copy Pose From Mesh() {bCopyCurves=true, bUseAttachedParent=true})
    Alpha=1
    ExternalForce="0,0,0"
   ) {
    angularDampingOverride=1
    angularSpringConstant=1500
    bAngularSpring=true
    bGravityOverrideInSimSpace=true
    bOverrideAngularDamping=true
    bUsePlanarLimit=false
    gravityScale=0.35
    simulationSpace=RootRelative
    alphaBoolBlend {blendOption=Linear}
    boundBone {boneName=foot_l_shoelace_r_bow_01_dyn}
    physicsBodyDefinitions [{
      sphereCollisionRadius=0
      boundBone {boneName=foot_l_shoelace_r_bow_01_dyn}
      boxExtents=<10 20 20>
      constraintSetup {angularLimitsMax=<5 2 2>, angularLimitsMin=<-5 -25 -25>, angularTarget=<0 0 0>}
      localJointOffset=<10 0 0>
     }]
    relativeSpaceBone {boneName=foot_l}
   }
   n2 = Anim Dynamics - Bone: foot_l_shoelace_r_bow_02_dyn(ComponentPose=@n1, Alpha=1, ExternalForce="0,0,0") {
    angularDampingOverride=1
    angularSpringConstant=1500
    bAngularSpring=true
    bGravityOverrideInSimSpace=true
    bOverrideAngularDamping=true
    bUsePlanarLimit=false
    gravityScale=0.35
    simulationSpace=RootRelative
    alphaBoolBlend {blendOption=Linear}
    boundBone {boneName=foot_l_shoelace_r_bow_02_dyn}
    physicsBodyDefinitions [{
      sphereCollisionRadius=0
      boundBone {boneName=foot_l_shoelace_r_bow_02_dyn}
      boxExtents=<10 20 20>
      constraintSetup {angularLimitsMax=<5 2 2>, angularLimitsMin=<-5 -25 -25>, angularTarget=<0 0 0>}
      localJointOffset=<10 0 0>
     }]
    relativeSpaceBone {boneName=foot_l}
   }
   n3 = Anim Dynamics - Bone: foot_l_shoelace_r_strand_01_dyn(ComponentPose=@n2, Alpha=1, ExternalForce="0,0,0") {
    angularDampingOverride=1
    angularSpringConstant=1500
    bAngularSpring=true
    bGravityOverrideInSimSpace=true
    bOverrideAngularDamping=true
    bUsePlanarLimit=false
    gravityScale=0.35
    simulationSpace=RootRelative
    alphaBoolBlend {blendOption=Linear}
    boundBone {boneName=foot_l_shoelace_r_strand_01_dyn}
    physicsBodyDefinitions [{
      sphereCollisionRadius=0
      boundBone {boneName=foot_l_shoelace_r_strand_01_dyn}
      boxExtents=<10 20 20>
      constraintSetup {angularLimitsMax=<5 0 0>, angularLimitsMin=<-5 -25 -25>, angularTarget=<0 0 0>}
      localJointOffset=<10 0 0>
     }]
    relativeSpaceBone {boneName=foot_l}
   }
   n4 = Anim Dynamics - Bone: foot_l_shoelace_r_strand_02_dyn(ComponentPose=@n3, Alpha=1, ExternalForce="0,0,0") {
    angularDampingOverride=1
    angularSpringConstant=1500
    bAngularSpring=true
    bGravityOverrideInSimSpace=true
    bOverrideAngularDamping=true
    bUsePlanarLimit=false
    gravityScale=0.35
    simulationSpace=RootRelative
    alphaBoolBlend {blendOption=Linear}
    boundBone {boneName=foot_l_shoelace_r_strand_02_dyn}
    physicsBodyDefinitions [{
      sphereCollisionRadius=0
      boundBone {boneName=foot_l_shoelace_r_strand_02_dyn}
      boxExtents=<10 20 20>
      constraintSetup {angularLimitsMax=<5 0 0>, angularLimitsMin=<-5 -25 -25>, angularTarget=<0 0 0>}
      localJointOffset=<10 0 0>
     }]
    relativeSpaceBone {boneName=foot_l}
   }
   n5 = Anim Dynamics - Bone: foot_l_shoelace_l_bow_01_dyn(ComponentPose=@n4, Alpha=1, ExternalForce="0,0,0") {
    angularDampingOverride=1
    angularSpringConstant=1500
    bAngularSpring=true
    bGravityOverrideInSimSpace=true
    bOverrideAngularDamping=true
    bUsePlanarLimit=false
    gravityScale=0.35
    simulationSpace=RootRelative
    alphaBoolBlend {blendOption=Linear}
    boundBone {boneName=foot_l_shoelace_l_bow_01_dyn}
    physicsBodyDefinitions [{
      sphereCollisionRadius=0
      boundBone {boneName=foot_l_shoelace_l_bow_01_dyn}
      boxExtents=<10 20 20>
      constraintSetup {angularLimitsMax=<5 2 2>, angularLimitsMin=<-5 -25 -25>, angularTarget=<0 0 0>}
      localJointOffset=<10 0 0>
     }]
    relativeSpaceBone {boneName=foot_l}
   }
   n6 = Anim Dynamics - Bone: foot_l_shoelace_l_bow_02_dyn(ComponentPose=@n5, Alpha=1, ExternalForce="0,0,0") {
    angularDampingOverride=1
    angularSpringConstant=1500
    bAngularSpring=true
    bGravityOverrideInSimSpace=true
    bOverrideAngularDamping=true
    bUsePlanarLimit=false
    gravityScale=0.35
    simulationSpace=RootRelative
    alphaBoolBlend {blendOption=Linear}
    boundBone {boneName=foot_l_shoelace_l_bow_02_dyn}
    physicsBodyDefinitions [{
      sphereCollisionRadius=0
      boundBone {boneName=foot_l_shoelace_l_bow_02_dyn}
      boxExtents=<10 20 20>
      constraintSetup {angularLimitsMax=<5 2 2>, angularLimitsMin=<-5 -25 -25>, angularTarget=<0 0 0>}
      localJointOffset=<10 0 0>
     }]
    relativeSpaceBone {boneName=foot_l}
   }
   n7 = Anim Dynamics - Bone: foot_l_shoelace_l_strand_01_dyn(ComponentPose=@n6, Alpha=1, ExternalForce="0,0,0") {
    angularDampingOverride=1
    angularSpringConstant=1500
    bAngularSpring=true
    bGravityOverrideInSimSpace=true
    bOverrideAngularDamping=true
    bUsePlanarLimit=false
    gravityScale=0.35
    simulationSpace=RootRelative
    alphaBoolBlend {blendOption=Linear}
    boundBone {boneName=foot_l_shoelace_l_strand_01_dyn}
    physicsBodyDefinitions [{
      sphereCollisionRadius=0
      boundBone {boneName=foot_l_shoelace_l_strand_01_dyn}
      boxExtents=<10 20 20>
      constraintSetup {angularLimitsMax=<5 0 0>, angularLimitsMin=<-5 -25 -25>, angularTarget=<0 0 0>}
      localJointOffset=<10 0 0>
     }]
    relativeSpaceBone {boneName=foot_l}
   }
   n8 = Anim Dynamics - Bone: foot_l_shoelace_l_strand_02_dyn(ComponentPose=@n7, Alpha=1, ExternalForce="0,0,0") {
    angularDampingOverride=1
    angularSpringConstant=1500
    bAngularSpring=true
    bGravityOverrideInSimSpace=true
    bOverrideAngularDamping=true
    bUsePlanarLimit=false
    gravityScale=0.35
    simulationSpace=RootRelative
    alphaBoolBlend {blendOption=Linear}
    boundBone {boneName=foot_l_shoelace_l_strand_02_dyn}
    physicsBodyDefinitions [{
      sphereCollisionRadius=0
      boundBone {boneName=foot_l_shoelace_l_strand_02_dyn}
      boxExtents=<10 20 20>
      constraintSetup {angularLimitsMax=<5 0 0>, angularLimitsMin=<-5 -25 -25>, angularTarget=<0 0 0>}
      localJointOffset=<10 0 0>
     }]
    relativeSpaceBone {boneName=foot_l}
   }
   n9 = Anim Dynamics - Bone: foot_r_shoelace_r_bow_01_dyn(ComponentPose=@n8, Alpha=1, ExternalForce="0,0,0") {
    angularDampingOverride=0.9999
    angularSpringConstant=1500
    bAngularSpring=true
    bGravityOverrideInSimSpace=true
    bOverrideAngularDamping=true
    bUsePlanarLimit=false
    gravityScale=0.35
    simulationSpace=RootRelative
    alphaBoolBlend {blendOption=Linear}
    boundBone {boneName=foot_r_shoelace_r_bow_01_dyn}
    physicsBodyDefinitions [{
      sphereCollisionRadius=0
      boundBone {boneName=foot_r_shoelace_r_bow_01_dyn}
      boxExtents=<10 20 20>
      constraintSetup {angularLimitsMax=<5 2 2>, angularLimitsMin=<-5 -25 -25>, angularTarget=<0 0 0>}
      localJointOffset=<10 0 0>
     }]
    relativeSpaceBone {boneName=foot_r}
   }
   n10 = Anim Dynamics - Bone: foot_r_shoelace_r_bow_01_dyn(ComponentPose=@n9, Alpha=1, ExternalForce="0,0,0") {
    angularDampingOverride=0.9999
    angularSpringConstant=1500
    bAngularSpring=true
    bGravityOverrideInSimSpace=true
    bOverrideAngularDamping=true
    bUsePlanarLimit=false
    gravityScale=0.35
    simulationSpace=RootRelative
    alphaBoolBlend {blendOption=Linear}
    boundBone {boneName=foot_r_shoelace_r_bow_01_dyn}
    physicsBodyDefinitions [{
      sphereCollisionRadius=0
      boundBone {boneName=foot_r_shoelace_r_bow_01_dyn}
      boxExtents=<10 20 20>
      constraintSetup {angularLimitsMax=<5 2 2>, angularLimitsMin=<-5 -25 -25>, angularTarget=<0 0 0>}
      localJointOffset=<10 0 0>
     }]
    relativeSpaceBone {boneName=foot_r}
   }
   n11 = Anim Dynamics - Bone: foot_r_shoelace_r_strand_01_dyn(ComponentPose=@n10, Alpha=1, ExternalForce="0,0,0") {
    angularDampingOverride=0.9999
    angularSpringConstant=1500
    bAngularSpring=true
    bGravityOverrideInSimSpace=true
    bOverrideAngularDamping=true
    bUsePlanarLimit=false
    gravityScale=0.35
    simulationSpace=RootRelative
    alphaBoolBlend {blendOption=Linear}
    boundBone {boneName=foot_r_shoelace_r_strand_01_dyn}
    physicsBodyDefinitions [{
      sphereCollisionRadius=0
      boundBone {boneName=foot_r_shoelace_r_strand_01_dyn}
      boxExtents=<10 20 20>
      constraintSetup {angularLimitsMax=<5 0 0>, angularLimitsMin=<-5 -25 -25>, angularTarget=<0 0 0>}
      localJointOffset=<10 0 0>
     }]
    relativeSpaceBone {boneName=foot_r}
   }
   n12 = Anim Dynamics - Bone: foot_r_shoelace_r_strand_02_dyn(ComponentPose=@n11, Alpha=1, ExternalForce="0,0,0") {
    angularDampingOverride=0.9999
    angularSpringConstant=1500
    bAngularSpring=true
    bGravityOverrideInSimSpace=true
    bOverrideAngularDamping=true
    bUsePlanarLimit=false
    gravityScale=0.35
    simulationSpace=RootRelative
    alphaBoolBlend {blendOption=Linear}
    boundBone {boneName=foot_r_shoelace_r_strand_02_dyn}
    physicsBodyDefinitions [{
      sphereCollisionRadius=0
      boundBone {boneName=foot_r_shoelace_r_strand_02_dyn}
      boxExtents=<10 20 20>
      constraintSetup {angularLimitsMax=<5 0 0>, angularLimitsMin=<-5 -25 -25>, angularTarget=<0 0 0>}
      localJointOffset=<10 0 0>
     }]
    relativeSpaceBone {boneName=foot_r}
   }
   n13 = Anim Dynamics - Bone: foot_r_shoelace_l_bow_01_dyn(ComponentPose=@n12, Alpha=1, ExternalForce="0,0,0") {
    angularDampingOverride=0.9999
    angularSpringConstant=1500
    bAngularSpring=true
    bGravityOverrideInSimSpace=true
    bOverrideAngularDamping=true
    bUsePlanarLimit=false
    gravityScale=0.35
    simulationSpace=RootRelative
    alphaBoolBlend {blendOption=Linear}
    boundBone {boneName=foot_r_shoelace_l_bow_01_dyn}
    physicsBodyDefinitions [{
      sphereCollisionRadius=0
      boundBone {boneName=foot_r_shoelace_l_bow_01_dyn}
      boxExtents=<10 20 20>
      constraintSetup {angularLimitsMax=<5 2 2>, angularLimitsMin=<-5 -25 -25>, angularTarget=<0 0 0>}
      localJointOffset=<10 0 0>
     }]
    relativeSpaceBone {boneName=foot_r}
   }
   n14 = Anim Dynamics - Bone: foot_r_shoelace_l_bow_02_dyn(ComponentPose=@n13, Alpha=1, ExternalForce="0,0,0") {
    angularDampingOverride=0.9999
    angularSpringConstant=1500
    bAngularSpring=true
    bGravityOverrideInSimSpace=true
    bOverrideAngularDamping=true
    bUsePlanarLimit=false
    gravityScale=0.35
    simulationSpace=RootRelative
    alphaBoolBlend {blendOption=Linear}
    boundBone {boneName=foot_r_shoelace_l_bow_02_dyn}
    physicsBodyDefinitions [{
      sphereCollisionRadius=0
      boundBone {boneName=foot_r_shoelace_l_bow_02_dyn}
      boxExtents=<10 20 20>
      constraintSetup {angularLimitsMax=<5 2 2>, angularLimitsMin=<-5 -25 -25>, angularTarget=<0 0 0>}
      localJointOffset=<10 0 0>
     }]
    relativeSpaceBone {boneName=foot_r}
   }
   n15 = Anim Dynamics - Bone: foot_r_shoelace_l_strand_01_dyn(ComponentPose=@n14, Alpha=1, ExternalForce="0,0,0") {
    angularDampingOverride=0.9999
    angularSpringConstant=1500
    bAngularSpring=true
    bGravityOverrideInSimSpace=true
    bOverrideAngularDamping=true
    bUsePlanarLimit=false
    gravityScale=0.35
    simulationSpace=RootRelative
    alphaBoolBlend {blendOption=Linear}
    boundBone {boneName=foot_r_shoelace_l_strand_01_dyn}
    physicsBodyDefinitions [{
      sphereCollisionRadius=0
      boundBone {boneName=foot_r_shoelace_l_strand_01_dyn}
      boxExtents=<10 20 20>
      constraintSetup {angularLimitsMax=<5 0 0>, angularLimitsMin=<-5 -25 -25>, angularTarget=<0 0 0>}
      localJointOffset=<10 0 0>
     }]
    relativeSpaceBone {boneName=foot_r}
   }
   n16 = Anim Dynamics - Bone: foot_r_shoelace_l_strand_01_dyn(ComponentPose=@n15, Alpha=1, ExternalForce="0,0,0") {
    angularDampingOverride=0.9999
    angularSpringConstant=1500
    bAngularSpring=true
    bGravityOverrideInSimSpace=true
    bOverrideAngularDamping=true
    bUsePlanarLimit=false
    gravityScale=0.35
    simulationSpace=RootRelative
    alphaBoolBlend {blendOption=Linear}
    boundBone {boneName=foot_r_shoelace_l_strand_01_dyn}
    physicsBodyDefinitions [{
      sphereCollisionRadius=0
      boundBone {boneName=foot_r_shoelace_l_strand_01_dyn}
      boxExtents=<10 20 20>
      constraintSetup {angularLimitsMax=<5 0 0>, angularLimitsMin=<-5 -25 -25>, angularTarget=<0 0 0>}
      localJointOffset=<10 0 0>
     }]
    relativeSpaceBone {boneName=foot_r}
   }
   Output Pose(Result=Component To Local(ComponentPose=@n16)) {layerGroup=None}
  ]}
 EventGraph {nodes [Event BlueprintUpdateAnimation()]}
]
variables []
